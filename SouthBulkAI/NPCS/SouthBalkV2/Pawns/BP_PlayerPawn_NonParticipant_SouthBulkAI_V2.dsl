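; Non-participant player pawn for the SouthBulk AI V2 NPC, a boss-style character whose weapon is kept out of sight.
; BeginPlay multicast-links a placeholder animation layer; death unlinks it and deactivates the Boss_SpecialActor component.
; The current weapon is hidden on unholster, and after a short delay on equip or customization completion.
; Hiding is event-driven, despite comments saying it runs every frame.
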(event EventBeginPlay
  (|CallParentFunction)
  ; TDIA-RM
  (Animation|MulticastLinkAnimLayer "/SouthBulkAI/Assets/Animations/SouthBalkAI_Layer_AnimBP_PLACEHOLDER.SouthBalkAI_Layer_AnimBP_PLACEHOLDER_C")
  (CallFunction|TryHideWeapon))

(event Collision|EventActorBeginOverlap
  (params (OtherActor :type Actor%20Object%20Reference))
  (|CallParentFunction OtherActor) ; This node is disabled and will not be called.\nDrag off pins to build functionality.
)

(event EventOnDied_Event
  (params (DamagedActor :type Actor%20Object%20Reference) (Damage :type Float%20%28single-precision%29) (InstigatedBy :type Controller%20Object%20Reference) (DamageCauser :type Actor%20Object%20Reference) (HitLocation :type Vector :default "0, 0, 0") (FHitComponent :type Primitive%20Component%20Object%20Reference) (BoneName :type Name :default "None") (Momentum :type Vector :default "0, 0, 0"))
  ; On Death Deactivate Special Actor
  (|CallParentFunction DamagedActor Damage InstigatedBy DamageCauser HitLocation FHitComponent BoneName Momentum)
  (Components|Activation|Deactivate (Variables|Default|GetBoss_SpecialActor))
  ; TDIA-RM
  (Animation|MulticastUnlinkAnimLayer "/SouthBulkAI/Assets/Animations/SouthBalkAI_Layer_AnimBP_PLACEHOLDER.SouthBalkAI_Layer_AnimBP_PLACEHOLDER_C"))

(event EventOnCharacterCustomizationCompleted
  (params (Pawn :type Fort%20Player%20Pawn%20Object%20Reference))
  (bind _returnvalue (Weapon|GetCurrentWeapon))
  ; Hide never-seen boss weapon every frame
  (Utilities|FlowControl|Delay 0.1)
  (Utilities|IsValid _returnvalue
    (:Is%20Valid
      (Rendering|SetActorHiddenInGame _returnvalue true))))

(event EventOnWeaponEquipped
  (params (NewWeapon :type Fort%20Weapon%20Object%20Reference) (PrevWeapon :type Fort%20Weapon%20Object%20Reference))
  ; Hide never-seen boss weapon every frame
  (Utilities|FlowControl|Delay 0.1)
  (Rendering|SetActorHiddenInGame NewWeapon true))

(event Custom|TryHideWeapon
  (EventDispatchers|BindEventtoOnUnholsteredEvent self (delegate Custom|WeaponUnholstered)))

(event Custom|WeaponUnholstered
  (Rendering|SetActorHiddenInGame (Weapon|GetCurrentWeapon self) true))

(fn ConstructionScript
  (|CallParentFunction))
